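# Sound cue for oil drum impacts in demolition, using a small melee attenuation and the impacts sound class.
# A quality-level switch plays one of three random impact waves at the two higher-indexed-first settings, or a single optimized wave at the other two.
# A modulator at the root varies playback, and concurrency limits cap simultaneous object impacts.
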SoundCue /Game/Sounds/Fort_Impact_Sounds/Demolition/OilDrum/Demo_OilDrum_Impact_Cue
root=Modulator_37
nodes [
 {id=Modulator_37, children [QualityLevel_1]}
 {id=QualityLevel_1, children [Random_0,Random_0,WavePlayer_3,WavePlayer_3]}
 {id=Random_0, children [WavePlayer_0,WavePlayer_1,WavePlayer_2], properties {weights [1,1,1]}}
 {id=WavePlayer_0, properties {soundWaveAssetPtr=&/oil_drum_impact_01}}
 {id=WavePlayer_1, properties {soundWaveAssetPtr=&/oil_drum_impact_02}}
 {id=WavePlayer_2, properties {soundWaveAssetPtr=&/oil_drum_impact_03}}
 {id=WavePlayer_3, properties {soundWaveAssetPtr=&/oil_drum_impact_03_opt}}
]
properties {
 attenuationSettings=&Fort_Weapon_Melee_Sml_Att
 maxDistance=2800
 soundClassObject=&Fort_Impacts
 volumeMultiplier=0.476
 concurrencyOverrides {maxCount=3, resolutionRule=StopOldest}
 concurrencySet [&CONC_Object_Impacts_FourMax]
}
